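; DeimosHUD_PROTO is a prototype HUD Blueprint for the Deimos mode that shows target mission progress and mission messages on screen.
; CreateMissionTracker spawns a tracker widget with remaining and total targets plus mission name, storing it so UpdateMissionTracker and RemoveMissionTracker can refresh or remove it.
; ShowMissionMessage creates a new message widget on each call, passing it text and a display time without keeping a reference.
; The event graph stubs and construction script are empty.
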(event EventBeginPlay)

(event Collision|EventActorBeginOverlap
  (params (OtherActor :type Actor%20Object%20Reference)))

(event EventTick
  (params (DeltaSeconds :type Float%20%28single-precision%29)))

(fn CreateMissionTracker
  (params (RemainingTargets :type Integer) (TotalTargets :type Integer) (MissionName :type String))
  (bind _returnvalue (UserInterface|CreateWidget "/Game/Athena/Deimos/HUD/Deimos_TargetMissionTracker.Deimos_TargetMissionTracker_C" (HUD|GetOwningPlayerController)))
  (Class|DeimosTargetMissionTracker|UpdateMessage _returnvalue RemainingTargets TotalTargets MissionName)
  (UserInterface|Viewport|AddtoViewport _returnvalue 0)
  (Variables|Default|SetMissionTrackerWidget _returnvalue))

(fn RemoveMissionTracker
  (Widget|RemovefromParent (Variables|Default|GetMissionTrackerWidget)))

(fn ShowMissionMessage
  (params (MessageTime :type Float%20%28double-precision%29) (MessageText :type String))
  (bind _returnvalue (UserInterface|CreateWidget "/Game/Athena/Deimos/HUD/Deimos_MissionMessage.Deimos_MissionMessage_C" (HUD|GetOwningPlayerController)))
  (Class|DeimosMissionMessage|UpdateMessage _returnvalue MessageText MessageTime)
  (UserInterface|Viewport|AddtoViewport _returnvalue 0))

(fn UpdateMissionTracker
  (params (RemainingTargets :type Integer) (TotalTargets :type Integer) (MissionName :type String))
  (Class|DeimosTargetMissionTracker|UpdateMessage (Variables|Default|GetMissionTrackerWidget) RemainingTargets TotalTargets MissionName))

(fn ConstructionScript)
