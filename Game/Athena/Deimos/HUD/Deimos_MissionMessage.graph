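# Deimos HUD widget that shows a timed mission message as large orange, outlined, centered text above the middle of the screen.
# UpdateMessage stores the given string in MessageString and passes the message length to the DelayedDestroy custom event.
# DelayedDestroy waits that long, then removes the widget from its parent.
# GetText_0 converts MessageString to Text, apparently as the binding for the single text block.
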WidgetBlueprint /Game/Athena/Deimos/HUD/Deimos_MissionMessage
graphs [
 EventGraph {nodes [
   n1 = DelayedDestroy!() -> (OutputDelegate, then, TimeToKill:double=~) {is=title K2Node_CustomEvent}
   > Delay(Duration=@n1.TimeToKill)
   > RemoveFromParent()
   Event PreConstruct()
   Event Construct()
   Event Tick()
  ]}
 GetText_0 {nodes [
   GetText_0() {is=title K2Node_FunctionEntry}
   > Return Node!(execute, ReturnValue:Text="To Text (String)"(InString=Get MessageString() -> (MessageString:String)))
  ]}
 UpdateMessage {nodes [
   n1 = UpdateMessage!() -> (then, String:String, MessageLength:double=~) {is=title K2Node_FunctionEntry}
   > Set MessageString(MessageString:String=@n1) -> (Output_Get:String)
   > DelayedDestroy!(execute, self, TimeToKill:double=@n1.MessageLength) -> (then) {is=title K2Node_CallFunction}
  ]}
]
variables [MessageString {type=string, default=""}]
widgets {
 info {inheritedWidgetCount=0, parentClass=&UserWidget, rootWidgetClass=&CanvasPanel, widgetCount=2}
 widgets [
  CanvasPanel_0 {widgetClassPath=&CanvasPanel, widget=CanvasPanel_0, properties {renderTransformPivot=<0 0>}}
  TextBlock_27 {
   widgetClassPath=&TextBlock
   parent=CanvasPanel_0
   widget=TextBlock_27
   properties {
    justification=Center
    text=TEST
    colorAndOpacity {colorUseRule=UseColor_Specified, specifiedColor=<rgba 1 0.5175 0 1>}
    font {
     bForceMonospaced=false
     bMaterialIsStencil=false
     fontMaterial=null
     fontObject=&Burbank
     letterSpacing=0
     monospacedWidth=1
     size=64
     skewAmount=0
     typefaceFontName=Burbank-Black
     outlineSettings {
      bApplyOutlineToDropShadows=true
      bMiteredCorners=false
      bSeparateFillAlpha=false
      outlineMaterial=null
      outlineSize=3
      outlineColor=<rgba 0 0 0 1>
     }
    }
   }
   slot {class=CanvasPanelSlot, properties {layoutData {alignment=<0 0>, anchors {maximum=<0.5 0.5>, minimum=<0.5 0.5>}, offsets=<ltrb 0 -425 0 0>}}}
  }
 ]
}
